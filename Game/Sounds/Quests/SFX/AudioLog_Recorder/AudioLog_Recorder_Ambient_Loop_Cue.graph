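# Sound cue for the looping ambient sound of the quest audio log recorder, in the Fort_Quests sound class.
# A mixer layers two looping waves, a base ambient loop and a shimmer loop, with both inputs boosted and the shimmer louder.
# It borrows the supply drop HiddenStar ambient loop attenuation and overrides concurrency to cap simultaneous instances at three.
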SoundCue /Game/Sounds/Quests/SFX/AudioLog_Recorder/AudioLog_Recorder_Ambient_Loop_Cue
root=Mixer_0
nodes [
 {id=Mixer_0, children [WavePlayer_0,WavePlayer_1], properties {inputVolume [1.65,2]}}
 {id=WavePlayer_0, properties {bLooping=true, soundWaveAssetPtr=&/AudioLog_Recorder_Ambient_Loop}}
 {
  id=WavePlayer_1
  properties {bLooping=true, soundWaveAssetPtr=&/AudioLog_Recorder_Ambient_Shimmer_Loop}
 }
]
properties {
 attenuationSettings=/Game/Athena/Sounds/SupplyDrops/Fort_HiddenStar_Ambient_Loop_Att
 bOverrideConcurrency=true
 maxDistance=2500
 soundClassObject=&Fort_Quests
 volumeMultiplier=1
 concurrencyOverrides {maxCount=3}
}
