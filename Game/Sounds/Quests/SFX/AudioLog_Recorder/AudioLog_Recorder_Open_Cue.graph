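# Sound cue for opening an audio log recorder, in the quest sound effects.
# It plays a single activation wave through one wave player with no randomization or mixing.
# It routes to the quest sound class with a boosted volume, uses a borrowed supply drop ambient loop attenuation, and overrides concurrency to cap simultaneous instances.
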SoundCue /Game/Sounds/Quests/SFX/AudioLog_Recorder/AudioLog_Recorder_Open_Cue
root=WavePlayer_1
nodes [{id=WavePlayer_1, properties {soundWaveAssetPtr=&/AudioLog_Recorder_Activate_01}}]
properties {
 attenuationSettings=/Game/Athena/Sounds/SupplyDrops/Fort_HiddenStar_Ambient_Loop_Att
 bOverrideConcurrency=true
 maxDistance=2500
 soundClassObject=&Fort_Quests
 volumeMultiplier=1.5
 concurrencyOverrides {maxCount=2}
}
